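# Idle glow effect for the DecaSphere pickaxe, with a single Whip_Glow emitter drawing soft blue sprites along the weapon's skeletal mesh.
# A burst spawns 17 persistent particles, whose spawn group picks color and bone or socket index.
# Particles follow the mesh each update, taking velocity from the sampled X axis.
# Time-of-day color scaling, gameplay distance fade and front-end overrides for size and alpha adjust the look.
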NiagaraSystem /BRCosmetics/Weapons/FORT_Melee/Pickaxe_DecaSphere/FX/NS_DecaSphere_Pickaxe_Idle
dependencies {
 usedDataInterfaces [
  &NiagaraDataInterfaceArrayColor
  &NiagaraDataInterfaceArrayInt32
  &NiagaraDataInterfaceSkeletalMesh
  &NiagaraDataInterfaceColorCurve
 ]
 usedDynamicInputs [&ColorFromCurve]
 usedModules [
  &self:SystemSpawnScript.SystemScriptSource.SystemScriptGraph.NiagaraNodeAssignment_1.SetVariables_#1
  &SystemState
  &self:SystemSpawnScript.SystemScriptSource.SystemScriptGraph.NiagaraNodeAssignment_2.SetVariables_#2
 ]
}
emitters {Whip_Glow {
  emitterUpdateScript [
   &EmitterState(
    Inactive Response=::NewEnumerator1
    Life Cycle Mode=::NewEnumerator1
    Loop Behavior=::NewEnumerator1
    Loop Duration Mode=::NewEnumerator1
    UseLoopDelay=false
   )
   &SpawnBurst_Instantaneous(Spawn Count=17, Use Spawn Probability=false, UseLoopCountLimit=false)
  ]
  particleSpawnScript [
   &FN_SetSpawnGroup(Number of Spawn Groups=17)
   &InitializeParticle(
    Color=&SelectLinearColorFromArray(
     Array Sampling Mode=::NewEnumerator1
     Direct Array Index=$Particles.SpawnGroup
     Color Selection Array=DataInterface() {propertyValues {ColorData [<rgba 0.2451 0.648 1 0.8>, <rgba 0.2451 0.648 1 0.5>]}}
     Random Seed=?
    )
    Lifetime=5
    Sprite Size=&MaskVector2dBySpawnGroup(
     Masked Vector2D=&FN_FrontEndVector2Modifier(Front End:Vector2f=<12 17>, Gameplay:Vector2f=<10 15>)
     Passthrough Vector2D=<10 18>
     Spawn Group Mask=1
    )
    Sprite Size Mode=::NewEnumerator4
    UsePositionOffset=false
    Write Material Random=true
   )
  ]
  particleUpdateScript [
   &SolveForcesAndVelocity(Clamp Velocity=false, Limit Acceleration=false)
   &ParticleState(Kill Particles When Lifetime Has Elapsed=false, Let Infinitely Lived Particles Die When Emitter Deactivates=true)
   &SkeletalMeshLocation(
    Bone / Socket Index=&SelectIntFromArray(
     Direct Array Index=$Particles.SpawnGroup
     Int Selection Array=DataInterface() {propertyValues {IntData [0,1,2,3,4,5,6,7,8,9,10,11,12,13,14,15,17]}}
    )
    MaskSpawnGroup=false
    Mesh Sampling Type=::NewEnumerator5
    Skeletal Mesh=$System.Skeletal Mesh
    Socket Sampling Mode=::NewEnumerator5
   )
   /SetVariables(Particles.Velocity=&ReturnMeshOrientationAxis(Mesh Orientation Axis=::XAxis, Mesh Orientation Quat=$Particles.SkeletalMeshLocation.SampledRotation)) {isSetParameters=true}
   &ScaleColor(
    Scale Alpha=&FN_FrontEndFloatModifier(
     Gameplay=&Multiply_Float(
      A=0.55
      B=&FN_LOD_Distance_Remap(Start Distance=300, End Distance=800, Minimum Output=1, Maximum Output=0, Clamp Output=true)
     )
     Front End=0.5
    )
    Scale RGB=&MakeVectorFromLinearColorRGB(Color=$System.TODColorMultiplier)
    ScaleA=true
    ScaleRGB=true
   )
   &CameraOffset(Camera Offset Amount=2)
  ]
  rendererClasses [&NiagaraSpriteRendererProperties]
  renderers [{rendererClass=&NiagaraSpriteRendererProperties}]
 }}
summary {
 systemName=NS_DecaSphere_Pickaxe_Idle
 emitters [{emitterName=Whip_Glow, rendererClasses [&NiagaraSpriteRendererProperties]}]
 userVariables [
  User.bIsFrontEnd {type=NiagaraBool}
  User.bIsFrontEndPreview {type=NiagaraBool}
  User.SwingSectionIndex {type=NiagaraInt32}
 ]
}
