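# Niagara system for the DecaSphere pickaxe swing effect, drawing a short-lived anim trail from a skeletal mesh.
# A single Trail emitter spawns, initializes and updates particles through shared AnimTrail modules, feeding two ribbon renderers.
# It sets a separate blue-tinted, half-width motion trail color and width alongside the main ribbon values.
# Both colors follow a time-of-day multiplier, with alpha falling along curves scaled by a system TrailFade value.
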NiagaraSystem /BRCosmetics/Weapons/FORT_Melee/Pickaxe_DecaSphere/FX/NS_DecaSphere_Pickaxe_Swing
dependencies {
 usedDataInterfaces [&NiagaraDataInterfaceCurve, &NiagaraDataInterfaceSkeletalMesh, &NiagaraDataInterfaceColorCurve]
 usedDynamicInputs [
  &self:SectionIndexFloatSelector_01
  &ColorFromCurve
 ]
 usedModules [
  &self:SystemSpawnScript.SystemScriptSource.SystemScriptGraph.NiagaraNodeAssignment_1.SetVariables_#1
  &SystemState
  &self:SystemSpawnScript.SystemScriptSource.SystemScriptGraph.NiagaraNodeAssignment_2.SetVariables_#2
  &self:SystemSpawnScript.SystemScriptSource.SystemScriptGraph.NiagaraNodeAssignment_0.SetVariables_#3
 ]
}
emitters {Trail {
  emitterUpdateScript [
   &EmitterState()
   &FN_SpawnAnimTrail_v2()
  ]
  particleSpawnScript [
   /Game/Effects/Niagara/Modules/AnimTrail/FN_InitializeAnimTrail/FN_InitializeAnimTrailBoneIndecies()
   &FN_InitializeAnimTrail(AnimTrailWidth=20, Color=<rgba 1 1 1 0.45>, Lifetime=0.11, SkeletalMesh=$System.Skeletal Mesh) {module=FN_InitializeAnimTrail002}
  ]
  particleUpdateScript [
   &FN_UpdateAnimTrail(SkeletalMesh=$System.Skeletal Mesh)
   /SetVariables(
    Particles.MotionTrailColor:LinearColor=&MultiplyLinearColor(
     Linear Color=&MultiplyLinearColor(
      Linear Color=<rgba 0.0928 0.1569 1 1>
      Scale Factor=&MakeVector4_FromLinearColor(LinearColor=$System.TODColorMultiplier)
     )
     Scale Factor=&MakeVector4(
      X=1
      Y=1
      Z=1
      W=&FloatFromCurve(
       FloatCurve=DataInterface() {propertyValues {Curve {keys [{arriveTangent=-0.5419, interpMode=RCIM_Cubic, time=0, value=0.35}, <at 1 0>]}}}
       Scale Curve=$System.TrailFade
      )
     )
    )
    Particles.MotionTrailWidth:NiagaraFloat=&Multiply_Float(A=$Particles.RibbonWidth, B=0.5)
   ) {isSetParameters=true}
   &ScaleColor(
    Scale Alpha=&Multiply_Float(
     A=&FloatFromCurve(
      FloatCurve=DataInterface() {propertyValues {Curve {keys [{arriveTangent=-0.5419, interpMode=RCIM_Cubic, time=0, value=0.25}, <at 1 0>]}}}
      Scale Curve=$System.TrailFade
     )
     B=1
    )
    Scale RGB=&MakeVectorFromLinearColorRGB(Color=$System.TODColorMultiplier)
    ScaleA=true
    ScaleRGB=true
   ) {module=ScaleColor001}
  ]
  rendererClasses [&NiagaraRibbonRendererProperties]
  renderers [
   {rendererClass=&NiagaraRibbonRendererProperties}
   {rendererClass=&NiagaraRibbonRendererProperties, rendererIndex=1}
  ]
 }}
summary {
 systemName=NS_DecaSphere_Pickaxe_Swing
 emitters [{emitterName=Trail, rendererClasses [&NiagaraRibbonRendererProperties]}]
 userVariables [
  User.RibbonIDOffset {type=NiagaraInt32}
  User.SwingSectionIndex {type=NiagaraInt32}
 ]
}
