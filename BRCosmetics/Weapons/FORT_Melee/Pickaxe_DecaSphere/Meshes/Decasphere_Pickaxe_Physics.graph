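# Physics asset for the DecaSphere pickaxe, a two-handed melee weapon with a jointed whip of simulated bodies hanging from its handle.
# Fourteen whip bodies chain from the handle with narrow swing limits and zero twist, small capsules ending in a large offset sphere-like tip.
# It also carries capsules for the character's thigh and calf bones, each leg pair joined by a free-rotating constraint and unconnected to the whip chain.
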PhysicsAsset /BRCosmetics/Weapons/FORT_Melee/Pickaxe_DecaSphere/Meshes/Decasphere_Pickaxe_Physics
bodies {
 Melee_TwoHanded_Handle [{shapeType=Capsule, length=1, radius=1}]
 calf_l [{shapeType=Capsule, length=50, radius=8, center=<-30 0 0>, rotation=<rot 90 0 0>}]
 calf_r [{shapeType=Capsule, length=50, radius=8, center=<30 0 0>, rotation=<rot 90 0 0>}]
 thigh_l [{shapeType=Capsule, length=40, radius=8, center=<-25 0 0>, rotation=<rot 90 0 0>}]
 thigh_r [{shapeType=Capsule, length=40, radius=8, center=<25 0 0>, rotation=<rot 90 0 0>}]
 whip_skin_01 [{shapeType=Capsule, length=2, radius=1.5, center=<0.05 0 0>, rotation=<rot 90 0 0>}]
 whip_skin_02 [{shapeType=Capsule, length=2, radius=1.5, center=<0.05 0 0>, rotation=<rot 90 0 0>}]
 whip_skin_03 [{shapeType=Capsule, length=2, radius=2, center=<0.05 0 0>, rotation=<rot 90 0 0>}]
 whip_skin_04 [{shapeType=Capsule, length=2, radius=2, center=<0.1 0 0>, rotation=<rot 90 0 0>}]
 whip_skin_05 [{shapeType=Capsule, length=2, radius=2, center=<0.1 0 0>, rotation=<rot 90 0 0>}]
 whip_skin_06 [{shapeType=Capsule, length=2, radius=2, center=<0.1 0 0>, rotation=<rot 90 0 0>}]
 whip_skin_07 [{shapeType=Capsule, length=2, radius=2, center=<0.1 0 0>, rotation=<rot 90 0 0>}]
 whip_skin_08 [{shapeType=Capsule, length=2, radius=2, center=<0.1 0 0>, rotation=<rot 90 0 0>}]
 whip_skin_09 [{shapeType=Capsule, length=2, radius=2, center=<0.1 0 0>, rotation=<rot 90 0 0>}]
 whip_skin_10 [{shapeType=Capsule, length=2, radius=2, center=<0.1 0 0>, rotation=<rot 90 0 0>}]
 whip_skin_11 [{shapeType=Capsule, length=2, radius=2, center=<0.1 0 0>, rotation=<rot 90 0 0>}]
 whip_skin_12 [{shapeType=Capsule, length=2, radius=2, center=<0.1 0 0>, rotation=<rot 90 0 0>}]
 whip_skin_13 [{shapeType=Capsule, length=2, radius=2, center=<0.1 0 0>, rotation=<rot 90 0 0>}]
 whip_skin_14 [{shapeType=Capsule, length=0, radius=7.2, center=<14.7 0 0>, rotation=<rot 90 0 0>}]
}
constraints [
 {bone1Name=whip_skin_01, bone2Name=Melee_TwoHanded_Handle, swing1LimitDegrees=20, swing2LimitDegrees=20, twistLimitDegrees=0}
 {bone1Name=whip_skin_02, bone2Name=whip_skin_01, swing1LimitDegrees=20, swing2LimitDegrees=20, twistLimitDegrees=0}
 {bone1Name=whip_skin_03, bone2Name=whip_skin_02, swing1LimitDegrees=20, swing2LimitDegrees=20, twistLimitDegrees=0}
 {bone1Name=whip_skin_04, bone2Name=whip_skin_03, swing1LimitDegrees=20, swing2LimitDegrees=20, twistLimitDegrees=0}
 {bone1Name=whip_skin_05, bone2Name=whip_skin_04, swing1LimitDegrees=20, swing2LimitDegrees=20, twistLimitDegrees=0}
 {bone1Name=whip_skin_06, bone2Name=whip_skin_05, swing1LimitDegrees=20, swing2LimitDegrees=20, twistLimitDegrees=0}
 {bone1Name=whip_skin_07, bone2Name=whip_skin_06, swing1LimitDegrees=20, swing2LimitDegrees=20, twistLimitDegrees=0}
 {bone1Name=whip_skin_08, bone2Name=whip_skin_07, swing1LimitDegrees=20, swing2LimitDegrees=20, twistLimitDegrees=0}
 {bone1Name=whip_skin_09, bone2Name=whip_skin_08, swing1LimitDegrees=20, swing2LimitDegrees=20, twistLimitDegrees=0}
 {bone1Name=whip_skin_10, bone2Name=whip_skin_09, swing1LimitDegrees=20, swing2LimitDegrees=20, twistLimitDegrees=0}
 {bone1Name=whip_skin_11, bone2Name=whip_skin_10, swing1LimitDegrees=20, swing2LimitDegrees=20, twistLimitDegrees=0}
 {bone1Name=whip_skin_12, bone2Name=whip_skin_11, swing1LimitDegrees=20, swing2LimitDegrees=20, twistLimitDegrees=0}
 {bone1Name=whip_skin_13, bone2Name=whip_skin_12, swing1LimitDegrees=20, swing2LimitDegrees=20, twistLimitDegrees=0}
 {bone1Name=whip_skin_14, bone2Name=whip_skin_13, swing1LimitDegrees=20, swing2LimitDegrees=20, twistLimitDegrees=0}
 {
  bone1Name=calf_l
  bone2Name=thigh_l
  swing1LimitDegrees=45
  swing1Motion=Free
  swing2LimitDegrees=45
  swing2Motion=Free
  twistLimitDegrees=45
  twistMotion=Free
 }
 {
  bone1Name=calf_r
  bone2Name=thigh_r
  swing1LimitDegrees=45
  swing1Motion=Free
  swing2LimitDegrees=45
  swing2Motion=Free
  twistLimitDegrees=45
  twistMotion=Free
 }
]
